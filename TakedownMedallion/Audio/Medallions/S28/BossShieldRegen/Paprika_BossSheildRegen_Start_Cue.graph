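# Sound cue for the start of the boss shield regeneration effect in the Season 28 Takedown Medallion audio set, classed as a gadget sound.
# A single wave player feeds a modulator that adds slight random pitch and volume variation.
# It plays quietly with Fortnite's ambient item loop attenuation, a short audible range, and debug enabled.
# A concurrency override limits simultaneous instances to two, with distance-based volume scaling and a brief release on stolen voices.
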SoundCue /TakedownMedallion/Audio/Medallions/S28/BossShieldRegen/Paprika_BossSheildRegen_Start_Cue
root=Modulator_0
nodes [
 {id=Modulator_0, children [WavePlayer_0], properties {pitchMax=1.005, pitchMin=1, volumeMax=1, volumeMin=0.88}}
 {id=WavePlayer_0, properties {soundWaveAssetPtr=&/Paprika_BossSheildRegen_Start}}
]
properties {
 attenuationSettings=/Game/Sounds/Attenuation/Fort_Item_Ambient_Loop_Mono_Default_Att
 bDebug=true
 bOverrideConcurrency=true
 maxDistance=1250
 soundClassObject=&Fort_Gadgets
 volumeMultiplier=0.2
 attenuationOverrides {falloffDistance=512, nonSpatializedRadiusStart=20, stereoSpread=420}
 concurrencyOverrides {maxCount=2, voiceStealReleaseTime=0.3, volumeScale=0.4, volumeScaleAttackTime=0.1, volumeScaleMode=Distance}
}
